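; Behavior tree decorator for the FirePetal Wandering Boss that passes when the boss is off its main arena island.
; It requires the IsInArena gameplay tag on the pawn, then scans its arena island sensor volumes for the SyphonIdle zone tag.
; If any matches, it returns true when the pawn does not overlap the stored sensor, otherwise false.
; The loop never breaks, so the stored sensor is the last one iterated, not necessarily the tagged one.
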(fn PerformConditionCheckAI
  (params (OwnerController :type AIController%20Object%20Reference) (ControlledPawn :type Pawn%20Object%20Reference))
  (bind _asbp_player_pawn_fire_petal_wandering_boss (Utilities|Casting|CastToBP_PlayerPawn_FirePetal_WanderingBoss ControlledPawn)
    (:then
      (bind _output_get (Variables|Default|SetWanderingBossPawn _asbp_player_pawn_fire_petal_wandering_boss))
      (bind _bdidactorhavetags (Class|AthenaQuestFunctionLibrary|AthenaCheckActorHasTags _output_get "(GameplayTags=((TagName=\"FirePetalBosses.Demons.IsInArena\")))" true))
      (if _bdidactorhavetags
        (Variables|Default|SetbFoundSensor false)
        (bind _output_get_2 (Variables|Default|SetArena%20Island%20Sensor%20Array (Variables|/FirePetal_WanderingBoss/NPCs/WanderingBoss/Pawn/BP_PlayerPawn_FirePetal_WanderingBoss.BP_PlayerPawn_FirePetal_WanderingBoss_C|GetArenaIslandSensorArray (Variables|Default|GetWanderingBossPawn))))
        (bind _array_element (Utilities|Array|ForEachLoopwithBreak _output_get_2)
          (:LoopBody
            (Variables|Default|SetCurrentIslandSensor _array_element)
            (if (GameplayTags|HasTag (Variables|/Script/PlayspaceSystem.GameplayVolume|GetVolumeTags _array_element) "(TagName=\"Ability.WanderingBoss.SyphonIdle.Zone\")" true)
              (Variables|Default|SetbFoundSensor true)))
          (:Completed
            _))
        (if (Variables|Default|GetbFoundSensor)
          (return (not (Collision|IsOverlappingActor (Variables|Default|GetCurrentIslandSensor) (Variables|Default|GetWanderingBossPawn) :!class /Script/Engine.Actor)))
          (else
            (return false)))
        (else
          (return false))))
    (:CastFailed
      (return false))))
